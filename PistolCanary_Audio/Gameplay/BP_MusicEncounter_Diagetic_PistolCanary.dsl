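; Music encounter Blueprint that plays diegetic, phase-based music for the PistolCanary event, covering spire, tentacle and hero moments.
; Its execute-enum event maps the incoming byte to a music phase enum and selects that phase's sound.
; If music is turned on, InitiateNewMusicPhase triggers "FadeOut" on any current audio component.
; It then spawns the new sound at the owner of the PistolCanary-tagged encounter component and stores it for the next transition.
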(event MusicEncounter|EventExecuteEnumEvent
  (params (Music%20Encounter%20Component :type Music%20Encounter%20Component%20Base%20Object%20Reference) (Event%20Enum :type Byte))
  (CallFunction|InitiateNewMusicPhase (Utilities|Select :Index (Utilities|Enum|BytetoEnumPistolCanaryMusicPhasesV2 Event%20Enum) :NewEnumerator1 (Variables|MusicEncounter|GetMaskMakerSpireActivation) :NewEnumerator2 (Variables|MusicEncounter|GetTentacleSummon) :NewEnumerator3 (Variables|MusicEncounter|GetTentacleEmerge) :NewEnumerator4 (Variables|MusicEncounter|GetHeroIntroductions) :NewEnumerator5 (Variables|MusicEncounter|GetSingleTentacleCombatLP) :NewEnumerator6 (Variables|MusicEncounter|GetJuniorHeroMoment01) :NewEnumerator7 (Variables|MusicEncounter|GetJuniorHerosCombatLoop) :NewEnumerator8 (Variables|MusicEncounter|GetJuniorHeroVictory) :NewEnumerator9 (Variables|MusicEncounter|GetMomentOfCalm) :NewEnumerator10 (Variables|MusicEncounter|GetTripleTentacleCombat) :NewEnumerator11 (Variables|MusicEncounter|GetMainHeroMoment) :NewEnumerator12 (Variables|MusicEncounter|GetMainHeroSpiresSmash) :NewEnumerator13 (Variables|MusicEncounter|GetMainHeroEnd) :NewEnumerator14 (Variables|MusicEncounter|GetHeroToTentacleCombatTransition) :NewEnumerator15 (Variables|Default|GetCombat02Loop) :NewEnumerator16 (Variables|Default|GetJuniorHero02))))

(fn InitiateNewMusicPhase
  (params (Music :type Sound%20Base%20Object%20Reference))
  (Variables|Default|GetTurnOnMusic
    (:then
      (bind _pistolcanarymusicencounteraudio (Variables|Default|GetPistolCanaryMusicEncounterAudio)
        (:then
          (Audio|Parameter|ExecuteTriggerParameter _pistolcanarymusicencounteraudio "FadeOut"))
        (:else
          _))
      (bind _audio_component (MusicEncounter|SpawnMusicAtLocation Music (Transformation|GetActorLocation (Components|GetOwner (MusicEncounter|FindMusicEncounterComponentByTag "(TagName=\"PistolCanary\")")))))
      (Variables|Default|SetPistolCanaryMusicEncounterAudio _audio_component))
    (:else)))
