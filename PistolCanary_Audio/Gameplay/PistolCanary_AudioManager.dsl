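; Audio manager actor for the PistolCanary event, controlling music phases, mixes and day-specific portal ambience.
; Delta-gated setters toggle event, beam and hero control bus mixes, gated by player-in-volume and, for event and hero, scalable-float flags.
; Integer phases map to an enum sent through the global event router as music encounter events.
; The MaskMaker portal loop crossfades out and in per day, and EndPlay clears all mixes.
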(event EventBeginPlay
  (CallFunction|EventUpdate false)
  (Variables|Default|SetEventMixON false)
  (Variables|Default|SetHeroActiveMix false))

(event Collision|EventActorBeginOverlap
  (params (OtherActor :type Actor%20Object%20Reference)))

(event EventTick
  (params (DeltaSeconds :type Float%20%28single-precision%29)))

(event Custom|SetMusicPhase
  (params (MusicPhase :type Integer))
  (bind _updated_variable (Utilities|FlowControl|DeltaGateWildcard (Variables|Default|GetMusicPhase) MusicPhase)
    (:Success
      (CallFunction|InttoMusicPhase _updated_variable)
      (if (not (Utilities|Enum|Equal(Enum) (Variables|Default|GetCurrentMusicPhase) (Variables|Default|GetNewMusicPhase)))
        (bind _output_get (Variables|Default|SetCurrentMusicPhase (Variables|Default|GetNewMusicPhase)))
        (MusicEncounter|ExecuteMusicEncounterGameplayEventByEnum (GameplayEventRouter|GetGlobalEventRouter) _output_get self)))))

(event Custom|SetEventMixON
  (params (EventMixON :type Boolean))
  (bind _updated_variable (Utilities|FlowControl|DeltaGateWildcard (Variables|Default|GetEventMixON) EventMixON)
    (:Success
      (if _updated_variable
        (if (ScalableFloat|GetValueasBool (Variables|Default|GetHFPistolCanaryEventModulation))
          (bind _eventconrolbusmix (Variables|Default|GetEventConrolBusMix)
            (:then
              _)
            (:else))
          (else
            (bind _eventconrolbusmix_2 (Variables|Default|GetEventConrolBusMix)
              (:then
                (Variables|Default|GetIsPlayerInVolume
                  (:then
                    (if (not (Audio|Modulation|IsControlBusMixActive _eventconrolbusmix_2))
                      (Audio|Modulation|ActivateControlBusMix _eventconrolbusmix_2)
                      (Development|PrintString "Event Control Bus: Activated")
                      (Audio|PushSoundMixModifier (Variables|Default|GetPistolCanaryMixMod))))
                  (:else
                    _)))
              (:else))))
        (if (Audio|Modulation|IsControlBusMixActive _eventconrolbusmix)
          (Audio|Modulation|DeactivateControlBusMix _eventconrolbusmix)
          (Development|PrintString "Event Control Bus: Deactivated"))))))

(event Custom|SetControlBusBeamMix
  (params (ControlBusMixON :type Boolean))
  (bind _updated_variable (Utilities|FlowControl|DeltaGateWildcard (Variables|Default|GetControlBusBeamMix) ControlBusMixON)
    (:Success
      (if _updated_variable
        (bind _beamcontrolbusmix (Variables|Default|GetBeamControlBusMix)
          (:then
            (Variables|Default|GetIsPlayerInVolume
              (:then
                (if (not (Audio|Modulation|IsControlBusMixActive _beamcontrolbusmix))
                  (Audio|Modulation|ActivateControlBusMix _beamcontrolbusmix)
                  (Development|PrintString "Event Control Bus: Activated")))
              (:else
                _)))
          (:else))
        (else
          (bind _beamcontrolbusmix_2 (Variables|Default|GetBeamControlBusMix)
            (:then
              _)
            (:else))))
      (if (Audio|Modulation|IsControlBusMixActive _beamcontrolbusmix_2)
        (Audio|Modulation|DeactivateControlBusMix _beamcontrolbusmix_2)
        (Development|PrintString "Event Control Bus: Deactivated")))))

(event Custom|SetHeroActiveMix
  (params (ControlBusMixON :type Boolean))
  (bind _updated_variable (Utilities|FlowControl|DeltaGateWildcard (Variables|Default|GetHeroActiveMix) ControlBusMixON)
    (:Success
      (if _updated_variable
        (if (ScalableFloat|GetValueasBool (Variables|Default|GetHFPistolCanaryHeroModulation))
          (bind _heroactivecontrolbusmix (Variables|Default|GetHeroActiveControlBusMix)
            (:then
              _)
            (:else))
          (else
            (bind _heroactivecontrolbusmix_2 (Variables|Default|GetHeroActiveControlBusMix)
              (:then
                (Variables|Default|GetIsPlayerInVolume
                  (:then
                    (if (not (Audio|Modulation|IsControlBusMixActive _heroactivecontrolbusmix_2))
                      (Audio|Modulation|ActivateControlBusMix _heroactivecontrolbusmix_2)
                      (Development|PrintString "Event Control Bus: Activated")))
                  (:else
                    _)))
              (:else))))
        (if (Audio|Modulation|IsControlBusMixActive _heroactivecontrolbusmix)
          (Audio|Modulation|DeactivateControlBusMix _heroactivecontrolbusmix)
          (Development|PrintString "Event Control Bus: Deactivated"))))))

(event Custom|PlayProgressivePortalAudio
  (params (EventDay :type Integer))
  (bind _returnvalue (Actor|GetActorOfClass "/BRMapCh6/FX/MaskFortress/BP/BP_MaskMaker_Portal.BP_MaskMaker_Portal_C"))
  (if (Audio|Components|Audio|IsPlaying (Variables|/BRMapCh6/FX/MaskFortress/BP/BP_MaskMaker_Portal.BP_MaskMaker_Portal_C|GetAmbientLoop _returnvalue))
    (Audio|Components|Audio|FadeOut (Variables|/BRMapCh6/FX/MaskFortress/BP/BP_MaskMaker_Portal.BP_MaskMaker_Portal_C|GetAmbientLoop _returnvalue) 0.5 0.0 "Linear" :!class /Script/Engine.AudioComponent))
  (Audio|Components|Audio|SetSound (Variables|/BRMapCh6/FX/MaskFortress/BP/BP_MaskMaker_Portal.BP_MaskMaker_Portal_C|GetAmbientLoop _returnvalue) (Utilities|Select :Index EventDay :Option%200 "/PistolCanary_Audio/Assets/Portal/SC_PistolCanary_Portal_Day01.SC_PistolCanary_Portal_Day01" :Option%201 "/PistolCanary_Audio/Assets/Portal/SC_PistolCanary_Portal_Day02.SC_PistolCanary_Portal_Day02" :Option%202 "/PistolCanary_Audio/Assets/Portal/SC_PistolCanary_Portal_Day03.SC_PistolCanary_Portal_Day03"))
  (Audio|Components|Audio|FadeIn (Variables|/BRMapCh6/FX/MaskFortress/BP/BP_MaskMaker_Portal.BP_MaskMaker_Portal_C|GetAmbientLoop _returnvalue) 0.5 1.0 0.0 "Linear" :!class /Script/Engine.AudioComponent))

(event EventEndPlay
  (params (EndPlayReason :type EEndPlayReason%20Enum :default "Destroyed"))
  (CallFunction|EventUpdate false)
  (CallFunction|DeactivateAllPistolCanaryMixes))

(event Custom|EventStart
  (Variables|Default|SetEventActive true))

(fn ActivateControlBusMix
  (params (Mix :type Sound%20Control%20Bus%20Mix%20Object%20Reference))
  (if (not (ScalableFloat|GetValueasBool (Variables|Default|GetHFPistolCanaryEventModulation)))
    (Variables|Default|GetIsPlayerInVolume
      (:then
        (Audio|Modulation|ActivateControlBusMix Mix))
      (:else))))

(fn DeactivateAllPistolCanaryMixes
  (bind _heroactivecontrolbusmix (Variables|Default|GetHeroActiveControlBusMix))
  (bind _beamcontrolbusmix (Variables|Default|GetBeamControlBusMix))
  (bind _eventconrolbusmix (Variables|Default|GetEventConrolBusMix))
  (bind _herointroductions (Variables|Default|GetHeroIntroductions))
  (Utilities|FlowControl|Sequence
    (:then_0
      (if (Audio|Modulation|IsControlBusMixActive _beamcontrolbusmix)
        (Audio|Modulation|DeactivateControlBusMix _beamcontrolbusmix)))
    (:then_1
      (if (Audio|Modulation|IsControlBusMixActive _heroactivecontrolbusmix)
        (Audio|Modulation|DeactivateControlBusMix _heroactivecontrolbusmix)))
    (:then_2
      (if (Audio|Modulation|IsControlBusMixActive _eventconrolbusmix)
        (Audio|Modulation|DeactivateControlBusMix _eventconrolbusmix)))
    (:then_3
      (if (Audio|Modulation|IsControlBusMixActive _herointroductions)
        (Audio|Modulation|DeactivateControlBusMix _herointroductions)))
    (:then_4
      (Audio|PopSoundMixModifier (Variables|Default|GetPistolCanaryMixMod)))
    (:then_5
      (Development|PrintString "Deactivated All PistolCanary Mixes" true true "(R=0.000000,G=0.660000,B=1.000000,A=1.000000)" 1e+01))))

(fn EventUpdate
  (params (bIsActive :type Boolean))
  (CallFunction|SetEventMixON bIsActive)
  (CallFunction|SetHeroActiveMix bIsActive)
  (CallFunction|SetControlBusBeamMix bIsActive)
  (Variables|Default|SetMusicPhase 0))

(fn ForceMixUpdate
  (Utilities|FlowControl|Sequence
    (:then_0
      (Variables|Default|GetEventMixON
        (:then
          (bind _eventconrolbusmix (Variables|Default|GetEventConrolBusMix)
            (:then
              (if (not (Audio|Modulation|IsControlBusMixActive _eventconrolbusmix))
                (Audio|Modulation|ActivateControlBusMix _eventconrolbusmix)))
            (:else)))
        (:else
          (bind _eventconrolbusmix_2 (Variables|Default|GetEventConrolBusMix)
            (:then
              (if (Audio|Modulation|IsControlBusMixActive _eventconrolbusmix_2)
                (Audio|Modulation|DeactivateControlBusMix _eventconrolbusmix_2)))
            (:else)))))
    (:then_1
      (Variables|Default|GetHeroActiveMix
        (:then
          (bind _heroactivecontrolbusmix (Variables|Default|GetHeroActiveControlBusMix)
            (:then
              (if (not (Audio|Modulation|IsControlBusMixActive _heroactivecontrolbusmix))
                (Audio|Modulation|ActivateControlBusMix _heroactivecontrolbusmix)))
            (:else)))
        (:else
          (bind _heroactivecontrolbusmix_2 (Variables|Default|GetHeroActiveControlBusMix)
            (:then
              (if (Audio|Modulation|IsControlBusMixActive _heroactivecontrolbusmix_2)
                (Audio|Modulation|DeactivateControlBusMix _heroactivecontrolbusmix_2)))
            (:else)))))))

(fn IntToMusicPhase
  (params (Selection :type Integer))
  (switch int Selection
    (:0
      (Variables|Default|SetNewMusicPhase "NewEnumerator0"))
    (:1
      (Variables|Default|SetNewMusicPhase "NewEnumerator1"))
    (:3
      (Variables|Default|SetNewMusicPhase "NewEnumerator3"))
    (:4
      (Variables|Default|SetNewMusicPhase "NewEnumerator4"))
    (:5
      (Variables|Default|SetNewMusicPhase "NewEnumerator5"))
    (:6
      (Variables|Default|SetNewMusicPhase "NewEnumerator6"))
    (:7
      (Variables|Default|SetNewMusicPhase "NewEnumerator7"))
    (:8
      (Variables|Default|SetNewMusicPhase "NewEnumerator16"))
    (:9
      (Variables|Default|SetNewMusicPhase "NewEnumerator15"))
    (:10
      (Variables|Default|SetNewMusicPhase "NewEnumerator8"))
    (:11
      (Variables|Default|SetNewMusicPhase "NewEnumerator9"))
    (:12
      (Variables|Default|SetNewMusicPhase "NewEnumerator10"))
    (:13
      (Variables|Default|SetNewMusicPhase "NewEnumerator11"))
    (:14
      (Variables|Default|SetNewMusicPhase "NewEnumerator12"))))

(fn SetEventDay
  (params (PistolCanaryEventDay :type Integer)))

(fn SetIsPlayerInVolume
  (params (Condition :type Boolean))
  (if Condition
    (Variables|Default|SetIsPlayerInVolume true)
    (else
      (Variables|Default|SetIsPlayerInVolume false))))

(fn ConstructionScript)
