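; Event audio dispatcher Blueprint that relays listener volume enter and exit to the Pistol Canary audio manager.
; BeginPlay runs InitializeAudio, which finds the manager with GetActorOfClass and caches it.
; ListenerEnterVolume and ListenerExitVolume store the in-volume flag and copy it to the manager.
; Entering then forces a mix update, exiting deactivates all Pistol Canary mixes, and Tick is empty.
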(event EventBeginPlay
  (CallFunction|InitializeAudio))

(event EventTick
  (params (DeltaSeconds :type Float%20%28single-precision%29)))

(event Custom|ListenerEnterVolume
  (bind _pistol_canary_audio_manager (Variables|Default|GetPistol%20Canary%20Audio%20Manager))
  (bind _output_get (Variables|Default|SetPlayerInVolume true))
  (Class|PistolCanaryAudioManager|SetIsPlayerinVolume _pistol_canary_audio_manager _output_get :!class /PistolCanary_Audio/Gameplay/PistolCanary_AudioManager.PistolCanary_AudioManager_C)
  (Class|PistolCanaryAudioManager|ForceMixUpdate _pistol_canary_audio_manager))

(event Custom|ListenerExitVolume
  (bind _pistol_canary_audio_manager (Variables|Default|GetPistol%20Canary%20Audio%20Manager))
  (bind _output_get (Variables|Default|SetPlayerInVolume false))
  (Class|PistolCanaryAudioManager|SetIsPlayerinVolume _pistol_canary_audio_manager _output_get :!class /PistolCanary_Audio/Gameplay/PistolCanary_AudioManager.PistolCanary_AudioManager_C)
  (Class|PistolCanaryAudioManager|DeactivateAllPistolCanaryMixes _pistol_canary_audio_manager))

(fn InitializeAudio
  (bind _returnvalue (Actor|GetActorOfClass "/PistolCanary_Audio/Gameplay/PistolCanary_AudioManager.PistolCanary_AudioManager_C"))
  (Variables|Default|SetPistol%20Canary%20Audio%20Manager _returnvalue))
